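# Physics asset for the CountyFrog pickaxe, a two-handed melee weapon cosmetic, defining simple physics shapes for its skeletal mesh.
# A short capsule covers the handle bone and a longer, thinner capsule covers the dym_handle_01 bone.
# One constraint joins the two bodies with narrow swing and twist limits, so dym_handle_01 can move only slightly relative to the handle.
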PhysicsAsset /BRCosmetics/Weapons/FORT_Melee/Pickaxe_CountyFrog/Meshes/CountyFrog_Axe_Physics
bodies {
 Melee_TwoHanded_Handle [{shapeType=Capsule, length=12, radius=1.85, shapeName=Melee_TwoHanded_Handle_capsule, center=<0 0 1.8813>, rotation=<rot 0 -90 0>}]
 dym_handle_01 [{shapeType=Capsule, length=35.3, radius=0.9, shapeName=dym_handle_01_capsule, rotation=<rot -90 -0.0003 0>}]
}
constraints [{bone1Name=dym_handle_01, bone2Name=Melee_TwoHanded_Handle, swing1LimitDegrees=10, swing2LimitDegrees=20, twistLimitDegrees=2}]
